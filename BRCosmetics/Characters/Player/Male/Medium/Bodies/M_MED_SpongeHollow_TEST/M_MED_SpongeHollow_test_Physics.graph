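# Test physics asset for the SpongeHollow medium male player body, giving physics bodies to a ring of dynamic stomach bones around the pelvis.
# Each stomach capsule is constrained to the pelvis capsule with angular limits that are widest at front and back center and narrower around the sides.
# The two front side bones also link tightly to the front middle bone, coupling their motion.
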PhysicsAsset /BRCosmetics/Characters/Player/Male/Medium/Bodies/M_MED_SpongeHollow_TEST/M_MED_SpongeHollow_test_Physics
bodies {
 dfrm_pelvis [{shapeType=Capsule, length=10, radius=10, center=<-1.6745 -0.6374 -0.1531>, rotation=<rot 0 -90 0>}]
 dyn_stomach_01_l [{
   shapeType=Capsule
   length=6
   radius=14
   shapeName=dyn_stomach_01_l_capsule
   center=<4.8966 19.7774 -14.0801>
   rotation=<rot -34.9637 87.9471 0>
  }]
 dyn_stomach_01_r [{
   shapeType=Capsule
   length=6
   radius=14
   shapeName=dyn_stomach_01_r_capsule
   center=<4.8966 19.7774 14.0801>
   rotation=<rot 34.9637 87.9471 0>
  }]
 dyn_stomach_02_l [{
   shapeType=Capsule
   length=6
   radius=14
   shapeName=dyn_stomach_02_l_capsule
   center=<4.2897 8.8999 -21.3962>
   rotation=<rot -66.6062 88.001 -0.2803>
  }]
 dyn_stomach_02_r [{
   shapeType=Capsule
   length=6
   radius=14
   shapeName=dyn_stomach_02_r_capsule
   center=<4.2901 8.8995 21.3965>
   rotation=<rot 66.6062 88.001 0.2803>
  }]
 dyn_stomach_03_l [{
   shapeType=Capsule
   length=6
   radius=14
   shapeName=dyn_stomach_03_l_capsule
   center=<5.0699 -4.6784 -23.1056>
   rotation=<rot -82.11 -92.0529 -180>
  }]
 dyn_stomach_03_r [{
   shapeType=Capsule
   length=6
   radius=14
   shapeName=dyn_stomach_03_r_capsule
   center=<5.0699 -4.6784 23.106>
   rotation=<rot 82.11 -92.0529 180>
  }]
 dyn_stomach_07_l [{
   shapeType=Capsule
   length=6
   radius=14
   shapeName=dyn_stomach_07_l_capsule
   center=<4.7826 -17.9006 -15.3297>
   rotation=<rot -28.4004 -92.3164 -179.8735>
  }]
 dyn_stomach_07_r [{
   shapeType=Capsule
   length=6
   radius=14
   shapeName=dyn_stomach_07_r_capsule
   center=<4.7825 -17.9006 15.3297>
   rotation=<rot 28.4004 -92.3164 179.8735>
  }]
 dyn_stomach_mid [{shapeType=Capsule, length=6, radius=14, shapeName=dyn_stomach_mid_capsule, center=<5.8379 25.2703 0>, rotation=<rot 0 -90 0>}]
 dyn_stomach_mid_back [{shapeType=Capsule, length=6, radius=14, shapeName=dyn_stomach_mid_back_capsule, center=<4.9541 -23.2984 0>, rotation=<rot 0 -92.2792 180>}]
}
constraints [
 {bone1Name=dyn_stomach_01_l, bone2Name=dfrm_pelvis, swing1LimitDegrees=60, swing2LimitDegrees=60, twistLimitDegrees=60}
 {bone1Name=dyn_stomach_01_r, bone2Name=dfrm_pelvis, swing1LimitDegrees=60, swing2LimitDegrees=60, twistLimitDegrees=60}
 {bone1Name=dyn_stomach_02_l, bone2Name=dfrm_pelvis, swing1LimitDegrees=55, swing2LimitDegrees=55, twistLimitDegrees=55}
 {bone1Name=dyn_stomach_02_r, bone2Name=dfrm_pelvis, swing1LimitDegrees=55, swing2LimitDegrees=55, twistLimitDegrees=55}
 {bone1Name=dyn_stomach_03_l, bone2Name=dfrm_pelvis, swing1LimitDegrees=45, swing2LimitDegrees=45, twistLimitDegrees=45}
 {bone1Name=dyn_stomach_03_r, bone2Name=dfrm_pelvis, swing1LimitDegrees=45, swing2LimitDegrees=45, twistLimitDegrees=45}
 {bone1Name=dyn_stomach_mid, bone2Name=dfrm_pelvis, swing1LimitDegrees=90, swing2LimitDegrees=90, twistLimitDegrees=90}
 {bone1Name=dyn_stomach_07_l, bone2Name=dfrm_pelvis, swing1LimitDegrees=45, swing2LimitDegrees=45, twistLimitDegrees=45}
 {bone1Name=dyn_stomach_07_r, bone2Name=dfrm_pelvis, swing1LimitDegrees=45, swing2LimitDegrees=45, twistLimitDegrees=45}
 {bone1Name=dyn_stomach_mid_back, bone2Name=dfrm_pelvis, swing1LimitDegrees=75, swing2LimitDegrees=75, twistLimitDegrees=75}
 {bone1Name=dyn_stomach_01_l, bone2Name=dyn_stomach_mid, swing1LimitDegrees=10, swing2LimitDegrees=10, twistLimitDegrees=10}
 {bone1Name=dyn_stomach_01_r, bone2Name=dyn_stomach_mid, swing1LimitDegrees=10, swing2LimitDegrees=10, twistLimitDegrees=10}
]
