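# Test post-process style AnimBP for the SpongeHollow male body cosmetic, adding deformation and jiggle physics on a pose copied from the parent mesh.
# A Control Rig, pelvis offset and arm twist/deform bone constraints produce a cached 'Post Deformation' pose.
# Two RigidBody sims (pelvis- and root-space) and a stomach constraint follow, bypassed when bPbBlendEnabled is true.
# Crouch, skydive and emote events swap sim-space settings; emotes also switch the active sim via alphas.
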AnimBlueprint /BRCosmetics/Characters/Player/Male/Medium/Bodies/M_MED_SpongeHollow_TEST/M_MED_SpongeHollow_AnimBP_test
target_skeleton=&/M_MED_SpongeHollow_Skeleton_test
graphs [
 AnimGraph {nodes [
   n1 = RigidBody!(
    ComponentPose=Local To Component(LocalPose=Use cached pose 'Post Deformation'())
    Alpha=Get Default Alpha() -> (Default Alpha:float)
    ExternalForce
    SimSpaceSettings:Sim Space Settings Structure=Get Sim Space Settings() -> (Sim Space Settings:Sim Space Settings Structure)
   ) -> (Pose) {
    bOverrideWorldGravity=true
    overridePhysicsAsset=&/M_MED_SpongeHollow_test_Physics
    simulationSpace=BaseBoneSpace
    baseBoneRef {boneName=pelvis}
    componentAppliedLinearAccClamp=<0 0 0>
    overrideWorldGravity {z=-1200}
    simSpaceSettings {dampingAlpha=0, maxAngularAcceleration=100, maxAngularVelocity=500, maxLinearAcceleration=100, maxLinearVelocity=1000, worldAlpha=1}
   }
   n2 = RigidBody!(ComponentPose=@n1, Alpha=Get Emote Alpha() -> (Emote Alpha:float), ExternalForce) -> (Pose) {
    bOverrideWorldGravity=true
    overridePhysicsAsset=&M_MED_SpongeHollow_test_Physics
    simulationSpace=BaseBoneSpace
    baseBoneRef {boneName=root}
    componentAppliedLinearAccClamp=<0 0 0>
    overrideWorldGravity {z=-1200}
    simSpaceSettings {dampingAlpha=0, maxAngularAcceleration=90, maxAngularVelocity=700, maxLinearAcceleration=90, maxLinearVelocity=1000, worldAlpha=1}
   }
   n3 = "Constraint - dfrm_stomach_lower_mid (P)"(ComponentPose=@n2, Alpha=0.5, ConstraintWeights_0=1) {boneToModify {boneName=dfrm_stomach_lower_mid}, constraintSetup [{targetBone {boneName=dfrm_stomach_mid}}], constraintWeights [1]}
   Output Pose(Result=Blend Poses by bool(
     BlendPose_0=Component To Local(ComponentPose=@n3)
     BlendPose_1=Use cached pose 'Post Deformation'()
     BlendTime_0=0
     BlendTime_1=0
     bActiveValue=Get bPbBlendEnabled() -> (bPbBlendEnabled:bool)
    )) {layerGroup=None}
   n4 = Control Rig!(
    Source=Copy Pose From Mesh() {bUseAttachedParent=true}
    Alpha
    isPlayingEmote_CR:bool=Get bIsPlayingEmote()
    isNotEmotingAlpha_CR:double=Get Is Not Emoting Alpha() -> (Is Not Emoting Alpha:double)
    isCrouching_CR:double=Get Is Crouching Alpha() -> (Is Crouching Alpha:double)
   ) -> (Pose) {controlRigAssetReference {blueprintRigClass=&/M_MED_SpongeHollow_CtrlRig.M_MED_SpongeHollow_CtrlRig_C}}
   n5 = "Transform (Modify) Bone - Bone: dfrm_pelvis"(ComponentPose=Local To Component(LocalPose=@n4), Alpha=1, Translation="3,0,0", Rotation="0.005,13.278,-0.009", Scale="1,1,1") {
    rotationMode=BMM_Additive
    rotationSpace=BCS_BoneSpace
    translationMode=BMM_Additive
    translationSpace=BCS_BoneSpace
    alphaBoolBlend {blendOption=Linear}
    alphaScaleBiasClamp {bInterpResult=true, interpSpeedIncreasing=6}
    boneToModify {boneName=dfrm_pelvis}
    rotation=<rot 0.0051 13.2779 -0.009>
    translation {x=3}
   }
   n6 = "Constraint - upperarm_twist_01_r (P)"!(ComponentPose=@n5, bAlphaBoolEnabled:bool=True, ConstraintWeights_0=1) -> (Pose) {
    alphaInputType=Bool
    alphaBoolBlend {blendOption=Linear}
    boneToModify {boneName=upperarm_twist_01_r}
    constraintSetup [{targetBone {boneName=upperarm_r}}]
    constraintWeights [1]
   }
   n7 = "Constraint - deform_upperarm_bk_base_r (P)"!(ComponentPose=@n6, bAlphaBoolEnabled:bool=True, ConstraintWeights_0=1) -> (Pose) {
    alphaInputType=Bool
    alphaBoolBlend {blendOption=Linear}
    boneToModify {boneName=deform_upperarm_bk_base_r}
    constraintSetup [{targetBone {boneName=upperarm_twist_01_r}}]
    constraintWeights [1]
   }
   n8 = "Constraint - deform_upperarm_bk_r (P)"!(ComponentPose=@n7, bAlphaBoolEnabled:bool=True, ConstraintWeights_0=1) -> (Pose) {
    alphaInputType=Bool
    alphaBoolBlend {blendOption=Linear}
    boneToModify {boneName=deform_upperarm_bk_r}
    constraintSetup [{targetBone {boneName=deform_upperarm_bk_base_r}}]
    constraintWeights [1]
   }
   n9 = "Constraint - deform_upperarm_fr_base_r (P)"!(ComponentPose=@n8, bAlphaBoolEnabled:bool=True, ConstraintWeights_0=1) -> (Pose) {
    alphaInputType=Bool
    alphaBoolBlend {blendOption=Linear}
    boneToModify {boneName=deform_upperarm_fr_base_r}
    constraintSetup [{targetBone {boneName=upperarm_twist_01_r}}]
    constraintWeights [1]
   }
   n10 = "Constraint - deform_upperarm_bk_r (P)"(ComponentPose=@n9, bAlphaBoolEnabled=True, ConstraintWeights_0=1) {
    alphaInputType=Bool
    alphaBoolBlend {blendOption=Linear}
    boneToModify {boneName=deform_upperarm_bk_r}
    constraintSetup [{targetBone {boneName=deform_upperarm_fr_r}}]
    constraintWeights [1]
   }
   n11 = "Constraint - upperarm_twist_02_r (P)"!(ComponentPose=@n10, bAlphaBoolEnabled:bool=True, ConstraintWeights_0=1) -> (Pose) {
    alphaInputType=Bool
    alphaBoolBlend {blendOption=Linear}
    boneToModify {boneName=upperarm_twist_02_r}
    constraintSetup [{targetBone {boneName=upperarm_r}}]
    constraintWeights [1]
   }
   n12 = "Constraint - deform_upperarm_up_base_r (P)"!(ComponentPose=@n11, bAlphaBoolEnabled:bool=True, ConstraintWeights_0=1) -> (Pose) {
    alphaInputType=Bool
    alphaBoolBlend {blendOption=Linear}
    boneToModify {boneName=deform_upperarm_up_base_r}
    constraintSetup [{targetBone {boneName=upperarm_twist_01_r}}]
    constraintWeights [1]
   }
   n13 = "Constraint - deform_upperarm_up_r (P)"!(ComponentPose=@n12, bAlphaBoolEnabled:bool=True, ConstraintWeights_0=1) -> (Pose) {
    alphaInputType=Bool
    alphaBoolBlend {blendOption=Linear}
    boneToModify {boneName=deform_upperarm_up_r}
    constraintSetup [{targetBone {boneName=deform_upperarm_up_base_r}}]
    constraintWeights [1]
   }
   n14 = "Constraint - upperarm_twist_01_l (P)"!(ComponentPose=@n13, bAlphaBoolEnabled:bool=True, ConstraintWeights_0=1) -> (Pose) {
    alphaInputType=Bool
    alphaBoolBlend {blendOption=Linear}
    boneToModify {boneName=upperarm_twist_01_l}
    constraintSetup [{targetBone {boneName=upperarm_l}}]
    constraintWeights [1]
   }
   n15 = "Constraint - deform_upperarm_bk_base_l (P)"!(ComponentPose=@n14, bAlphaBoolEnabled:bool=True, ConstraintWeights_0=1) -> (Pose) {
    alphaInputType=Bool
    alphaBoolBlend {blendOption=Linear}
    boneToModify {boneName=deform_upperarm_bk_base_l}
    constraintSetup [{targetBone {boneName=upperarm_l}}]
    constraintWeights [1]
   }
   n16 = "Constraint - deform_upperarm_bk_l (P)"!(ComponentPose=@n15, bAlphaBoolEnabled:bool=True, ConstraintWeights_0=1) -> (Pose) {
    alphaInputType=Bool
    alphaBoolBlend {blendOption=Linear}
    boneToModify {boneName=deform_upperarm_bk_l}
    constraintSetup [{targetBone {boneName=deform_upperarm_bk_l}}]
    constraintWeights [1]
   }
   n17 = "Constraint - deform_upperarm_fr_base_l (P)"!(ComponentPose=@n16, bAlphaBoolEnabled:bool=True, ConstraintWeights_0=1) -> (Pose) {
    alphaInputType=Bool
    alphaBoolBlend {blendOption=Linear}
    boneToModify {boneName=deform_upperarm_fr_base_l}
    constraintSetup [{targetBone {boneName=upperarm_l}}]
    constraintWeights [1]
   }
   n18 = "Constraint - deform_upperarm_fr_l (P)"!(ComponentPose=@n17, bAlphaBoolEnabled:bool=True, ConstraintWeights_0=1) -> (Pose) {
    alphaInputType=Bool
    alphaBoolBlend {blendOption=Linear}
    boneToModify {boneName=deform_upperarm_fr_l}
    constraintSetup [{targetBone {boneName=deform_upperarm_fr_base_l}}]
    constraintWeights [1]
   }
   n19 = "Constraint - deform_upperarm_up_base_l (P)"!(ComponentPose=@n18, bAlphaBoolEnabled:bool=True, ConstraintWeights_0=1) -> (Pose) {
    alphaInputType=Bool
    alphaBoolBlend {blendOption=Linear}
    boneToModify {boneName=deform_upperarm_up_base_l}
    constraintSetup [{targetBone {boneName=upperarm_l}}]
    constraintWeights [1]
   }
   n20 = "Constraint - deform_upperarm_up_l (P)"!(ComponentPose=@n19, bAlphaBoolEnabled:bool=True, ConstraintWeights_0=1) -> (Pose) {
    alphaInputType=Bool
    alphaBoolBlend {blendOption=Linear}
    boneToModify {boneName=deform_upperarm_up_l}
    constraintSetup [{targetBone {boneName=deform_upperarm_up_base_l}}]
    constraintWeights [1]
   }
   n21 = "Constraint - upperarm_twist_02_l (P)"!(ComponentPose=@n20, bAlphaBoolEnabled:bool=True, ConstraintWeights_0=1) -> (Pose) {
    alphaInputType=Bool
    alphaBoolBlend {blendOption=Linear}
    boneToModify {boneName=upperarm_twist_02_l}
    constraintSetup [{targetBone {boneName=upperarm_l}}]
    constraintWeights [1]
   }
   n22 = "Constraint - lowerarm_twist_01_l (P)"!(ComponentPose=@n21, bAlphaBoolEnabled:bool=True, ConstraintWeights_0=1) -> (Pose) {
    alphaInputType=Bool
    alphaBoolBlend {blendOption=Linear}
    boneToModify {boneName=lowerarm_twist_01_l}
    constraintSetup [{targetBone {boneName=lowerarm_l}}]
    constraintWeights [1]
   }
   n23 = "Constraint - lowerarm_twist_02_l (P)"!(ComponentPose=@n22, bAlphaBoolEnabled:bool=True, ConstraintWeights_0=1) -> (Pose) {
    alphaInputType=Bool
    alphaBoolBlend {blendOption=Linear}
    boneToModify {boneName=lowerarm_twist_02_l}
    constraintSetup [{targetBone {boneName=lowerarm_l}}]
    constraintWeights [1]
   }
   n24 = "Constraint - lowerarm_twist_02_r (P)"!(ComponentPose=@n23, bAlphaBoolEnabled:bool=True, ConstraintWeights_0=1) -> (Pose) {
    alphaInputType=Bool
    alphaBoolBlend {blendOption=Linear}
    boneToModify {boneName=lowerarm_twist_02_r}
    constraintSetup [{targetBone {boneName=lowerarm_r}}]
    constraintWeights [1]
   }
   n25 = "Constraint - lowerarm_twist_01_r (P)"!(ComponentPose=@n24, bAlphaBoolEnabled:bool=True, ConstraintWeights_0=1) -> (Pose) {
    alphaInputType=Bool
    alphaBoolBlend {blendOption=Linear}
    boneToModify {boneName=lowerarm_twist_01_r}
    constraintSetup [{targetBone {boneName=lowerarm_r}}]
    constraintWeights [1]
   }
   Save cached pose 'Post Deformation'(Pose=Component To Local(ComponentPose=@n25)) {cacheName=Post Deformation}
  ]}
 EventGraph {nodes [
   Event OnBeginCrouching()
   > Set Is Crouching Alpha(Is Crouching Alpha:double=1) -> (Output_Get:double)
   > Set Sim Space Settings(Sim Space Settings:Sim Space Settings Structure=Get Crouch Sim Space Settings() -> (Crouch Sim Space Settings:Sim Space Settings Structure)) -> (Output_Get:Sim Space Settings Structure)
   Event OnEndCrouching()
   > Set Is Crouching Alpha(Is Crouching Alpha=0)
   > Set Sim Space Settings(Sim Space Settings=Get Default Sim Space Settings() -> (Default Sim Space Settings:Sim Space Settings Structure))
   Event OnBeginSkydiving()
   > Set Sim Space Settings(Sim Space Settings=Get Skydive Sim Space Settings() -> (Skydive Sim Space Settings:Sim Space Settings Structure))
   Event OnEndSkydiving()
   > Set Sim Space Settings(Sim Space Settings=Get Default Sim Space Settings())
   Event OnBeginPlayingEmote() {is=title K2Node_Event}
   > Set Sim Space Settings(Sim Space Settings=Get Emote Sim Space Settings() -> (Emote Sim Space Settings:Sim Space Settings Structure))
   > Set Emote Alpha(Emote Alpha:float=0.7) -> (Output_Get:float)
   > Set Default Alpha(Default Alpha:float=0) -> (Output_Get:float)
   Event OnEndPlayingEmote() {is=title K2Node_Event}
   > Set Sim Space Settings(Sim Space Settings=Get Default Sim Space Settings())
   > Set Emote Alpha(Emote Alpha=0)
   > Set Default Alpha(Default Alpha=1)
   Event BlueprintUpdateAnimation()
  ]}
]
variables [
 Crouch Sim Space Settings {
  type=SimSpaceSettings
  default {
   dampingAlpha=0
   maxAngularAcceleration=10
   maxAngularVelocity=100
   maxLinearAcceleration=10
   maxLinearVelocity=500
   velocityScaleZ=0.7
   worldAlpha=0.7
   externalLinearDragV=<0 0 0>
  }
 }
 Default Alpha {type=number, default=1}
 Default Sim Space Settings {
  type=SimSpaceSettings
  default {
   dampingAlpha=1
   maxAngularAcceleration=80
   maxAngularVelocity=400
   maxLinearAcceleration=80
   maxLinearVelocity=800
   velocityScaleZ=0.9
   worldAlpha=0.9
   externalLinearDragV=<0 0 0>
  }
 }
 Disable Belly Jiggle {type=boolean}
 Emote Alpha {type=number}
 Emote Sim Space Settings {
  type=SimSpaceSettings
  default {
   dampingAlpha=0
   maxAngularAcceleration=10
   maxAngularVelocity=100
   maxLinearAcceleration=10
   maxLinearVelocity=500
   velocityScaleZ=1
   worldAlpha=1
   externalLinearDragV=<0 0 0>
  }
 }
 Is Crouching Alpha {type=number}
 Is Not Emoting Alpha {type=number}
 LOD 2 {type=boolean}
 Sim Space Settings {
  type=SimSpaceSettings
  default {
   dampingAlpha=0
   maxAngularAcceleration=100
   maxAngularVelocity=100
   maxLinearAcceleration=100
   maxLinearVelocity=1000
   velocityScaleZ=1
   worldAlpha=0.66
   externalLinearDragV=<0 0 0>
  }
 }
 Skydive Sim Space Settings {
  type=SimSpaceSettings
  default {
   dampingAlpha=0
   maxAngularAcceleration=10
   maxAngularVelocity=100
   maxLinearAcceleration=10
   maxLinearVelocity=500
   velocityScaleZ=0.3
   worldAlpha=0.3
   externalLinearDragV=<0 0 0>
  }
 }
]
